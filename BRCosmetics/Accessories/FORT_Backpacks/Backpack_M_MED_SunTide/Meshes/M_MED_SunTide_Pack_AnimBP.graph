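# Animation Blueprint for the SunTide backpack cosmetic, adjusting the pack to spine bend and making its two loose pieces float.
# Event graph maps spine_05-to-spine_01 delta yaw to pack roll and conditionally smoothed translation, generates sine bob offsets, and lowers dynamics alpha in frontend display.
# AnimGraph chains bone offsets from ref pose into spring-based AnimDynamics on both pieces, then adds upward offsets scaled by that alpha.
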AnimBlueprint /BRCosmetics/Accessories/FORT_Backpacks/Backpack_M_MED_SunTide/Meshes/M_MED_SunTide_Pack_AnimBP
target_skeleton=&/M_MED_SunTide_Pack_Skeleton
graphs [
 AnimGraph {nodes [
   n1 = "Transform (Modify) Bone - Bone: root"(ComponentPose=Local To Component(LocalPose=Local Space Ref Pose()), Translation="0,-4,4.5", Rotation="0,0,0") {
    rotationMode=BMM_Additive
    translationMode=BMM_Additive
    alphaBoolBlend {blendOption=Linear}
    boneToModify {boneName=root}
    translation {y=-4, z=4.5}
   }
   n2 = "Transform (Modify) Bone - Bone: dyn_piece_large"(ComponentPose=@n1, Alpha=1, Translation="-1,0,1", Rotation="0,0,0", Scale="1,1,1") {
    rotationMode=BMM_Additive
    translationMode=BMM_Additive
    alphaBoolBlend {blendOption=Linear}
    boneToModify {boneName=dyn_piece_large}
    translation {x=-1, z=1}
   }
   n3 = "Transform (Modify) Bone - Bone: pack"(ComponentPose=@n2, Translation=Get backpack_translation(), Rotation=Get backpack_rotate()) {rotationMode=BMM_Additive, translationMode=BMM_Additive, alphaBoolBlend {blendOption=Linear}, boneToModify {boneName=pack}}
   n4 = Get Floater() -> (Floater:Vector)
   n5 = Get dynAlpha() -> (dynAlpha:double)
   n6 = Anim Dynamics - Bone: dyn_piece_large(
    ComponentPose="Transform (Modify) Bone - Bone: dyn_piece_small"(
     ComponentPose="Transform (Modify) Bone - Bone: dyn_piece_large"(ComponentPose=@n3, Alpha=1, Translation=@n4, Rotation="0,0,0", Scale="1,1,1") {translationMode=BMM_Additive, alphaBoolBlend {blendOption=Linear}, boneToModify {boneName=dyn_piece_large}}
     Alpha=1
     Translation=Get Floater2() -> (Floater2:Vector)
     Rotation="0,0,0"
     Scale="1,1,1"
    ) {translationMode=BMM_Additive, alphaBoolBlend {blendOption=Linear}, boneToModify {boneName=dyn_piece_small}}
    Alpha=@n5
    ExternalForce="0,0,0"
   ) {
    bAngularSpring=true
    bGravityOverrideInSimSpace=true
    bLinearSpring=true
    bOverrideLinearDamping=true
    bUseGravityOverride=true
    bUseSphericalLimits=true
    linearDampingOverride=0.988
    linearSpringConstant=50
    alphaBoolBlend {blendOption=Linear}
    boundBone {boneName=dyn_piece_large}
    componentAppliedLinearAccClamp=<150000 10000 75000>
    componentLinearAccScale {x=0.04, z=0.05}
    componentLinearVelScale {x=0.04, z=0.05}
    gravityOverride {z=-980}
    physicsBodyDefinitions [{
      boundBone {boneName=dyn_piece_large}
      boxExtents=<25 25 25>
      constraintSetup {
       linearXLimitType=Free
       linearYLimitType=Free
       linearZLimitType=Free
       angularLimitsMax=<0 0 2>
       angularLimitsMin=<0 0 -2>
       angularTarget=<0 0 0>
       linearAxesMax=<1.25 1.25 1.25>
       linearAxesMin=<-1.25 0 -1.25>
      }
      localJointOffset=<0 0 0>
     }]
   }
   n7 = Anim Dynamics - Bone: dyn_piece_small(ComponentPose=@n6, Alpha=@n5, ExternalForce="0,0,0") {
    bAngularSpring=true
    bGravityOverrideInSimSpace=true
    bLinearSpring=true
    bOverrideLinearDamping=true
    bUseGravityOverride=true
    bUseSphericalLimits=true
    linearDampingOverride=0.988
    linearSpringConstant=50
    alphaBoolBlend {blendOption=Linear}
    boundBone {boneName=dyn_piece_small}
    componentAppliedLinearAccClamp=<150000 10000 75000>
    componentLinearAccScale {x=0.04, z=0.05}
    componentLinearVelScale {x=0.04, z=0.05}
    gravityOverride {z=-980}
    physicsBodyDefinitions [{
      boundBone {boneName=dyn_piece_small}
      boxExtents=<25 25 25>
      constraintSetup {
       linearXLimitType=Free
       linearYLimitType=Free
       linearZLimitType=Free
       angularLimitsMax=<0 0 2>
       angularLimitsMin=<0 0 -2>
       angularTarget=<0 0 0>
       linearAxesMax=<1.25 1.25 1.25>
       linearAxesMin=<-1.25 0 -1.25>
      }
      localJointOffset=<0 0 0>
     }]
   }
   n8 = "Transform (Modify) Bone - Bone: dyn_piece_large"(ComponentPose=@n7, Alpha=@n5, Translation="0.000,0.000,14", Rotation="0,0,0", Scale="1,1,1") {translationMode=BMM_Additive, alphaBoolBlend {blendOption=Linear}, boneToModify {boneName=dyn_piece_large}, translation {z=14}}
   n9 = "Transform (Modify) Bone - Bone: dyn_piece_small"(ComponentPose=@n8, Alpha=@n5, Translation="0.000,0.000,14", Rotation="0,0,0", Scale="1,1,1") {translationMode=BMM_Additive, alphaBoolBlend {blendOption=Linear}, boneToModify {boneName=dyn_piece_small}, translation {z=14}}
   Output Pose(Result=Component To Local(ComponentPose=@n9)) {layerGroup=None}
  ]}
 EventGraph {nodes [
   Event BlueprintUpdateAnimation()
   > n1 = Branch(Condition=Get bIsInFrontEnd())
   n2 = GetOwnerSkeletalMeshComponent()
   n3 = GetParentBone!(self:Skinned Mesh Component Object Reference=@n2, BoneName:Name=spine_01) -> (ReturnValue:Name=None)
   > n4 = Branch(Condition="float < float"(
     A=GetSocketRotation!(self:Scene Component Object Reference=@n2, InSocketName:Name=@n3) -> (ReturnValue_Roll:float, ReturnValue_Pitch:float, ReturnValue_Yaw:float).ReturnValue_Pitch
     B=-30
    ))
   n5 = GetOwnerSkeletalMeshComponent()
   n6 = BreakRotator(InRot=BreakTransform(InTransform=GetDeltaTransformFromRefPose(self=@n5, BoneName=spine_05, BaseName=spine_01)).Rotation)
   n7 = MakeRotator(Roll=MapRangeClamped(Value=@n6.Yaw, InRangeA=0, InRangeB=-40, OutRangeB=35))
   Set backpack_rotate(execute=[@n4.then, @n1.else], backpack_rotate=@n7)
   n8 = GetDeltaSeconds!(self:Anim Instance Object Reference) -> (ReturnValue:float)
   n9 = MapRangeClamped(Value=@n6.Yaw, InRangeA=0, InRangeB=-1, OutRangeB=-4)
   n10 = Get!() -> (backpack_translation_X:double, backpack_translation_Y:double, backpack_translation_Z:double)
   > n11 = Set backpack_translation(backpack_translation=MakeVector(Y=FInterpTo(Current=@n10.backpack_translation_Y, Target=@n9, DeltaTime=@n8, InterpSpeed=0.5)))
   n12 = GetGameTimeInSeconds()
   n13 = "Sin (Radians)"(A=float * float(A=float * float(A=0.7, B=-1), B=@n12))
   n14 = "Sin (Radians)"(A=float * float(A=float * float(A=0.8, B=2), B=@n12))
   Set Floater(execute=[@n18.then, @n11.then], Floater:Vector=MakeVector(X=float * float(A=@n13, B=1.25), Y=0, Z=float * float(A=@n14, B=1.25))) -> (Output_Get:Vector)
   n15 = "Sin (Radians)"(A=float * float(A=float * float(A=0.9, B=3), B=@n12))
   n16 = "Sin (Radians)"(A=float * float(A=float * float(A=0.8, B=-1), B=@n12))
   > Set Floater2(Floater2:Vector=MakeVector(X=float * float(A=@n16, B=0.75), Y=0, Z=float * float(A=@n15, B=0.75))) -> (Output_Get:Vector)
   > n17 = Branch(Condition=IsCosmeticObjectDisplayedInFrontendContext!(CosmeticObject:Object Reference) -> (ReturnValue:bool))
   > Set dynAlpha(dynAlpha:double=0.175) -> (Output_Get:double)
   Set dynAlpha(execute=@n17.else, dynAlpha=1)
   n18 = Set backpack_rotate(execute=@n4.else, backpack_rotate=@n7)
   TryGetPawnOwner()
  ]}
]
variables [
 backpack_rotate {type=Rotator}
 backpack_translation {type=Vector}
 dynAlpha {type=number}
 Floater {type=Vector}
 Floater2 {type=Vector}
]
